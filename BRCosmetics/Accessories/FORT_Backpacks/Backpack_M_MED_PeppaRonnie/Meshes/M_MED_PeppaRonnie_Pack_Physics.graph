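# Physics asset for the PeppaRonnie pizza backpack cosmetic, defining collision bodies and joints for its pizza and dangling cheese drips.
# Boxes cover a four-bone pizza chain and both drip bases, while small capsules cover the drip ends.
# Tight swing and twist limits keep the pizza chain nearly rigid.
# Much looser limits let one drip hang from pizza3 and a forked two-ended drip hang from pizza4.
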PhysicsAsset /BRCosmetics/Accessories/FORT_Backpacks/Backpack_M_MED_PeppaRonnie/Meshes/M_MED_PeppaRonnie_Pack_Physics
bodies {
 cheeseDripABase [{shapeType=Box, extentX=2.2625, extentY=1.5337, extentZ=3.9398, center=<1.0708 -0.0982 -0.2688>}]
 cheeseDripAEnd [{shapeType=Capsule, length=1.505, radius=0.505, center=<1.1695 0 0>, rotation=<rot 90 -0.0197 -0.0197>}]
 cheeseDripB1End [{shapeType=Capsule, length=3.01, radius=0.5, center=<1.9004 0.0984 0.0159>, rotation=<rot 90 0.0149 0.0149>}]
 cheeseDripB2End [{shapeType=Capsule, length=1.26, radius=0.5, center=<0.9273 0.1265 0.0228>, rotation=<rot 90 -0.0197 -0.0197>}]
 cheeseDripBBBase [{shapeType=Box, extentX=2.8858, extentY=1.797, extentZ=6.1923, center=<1.2811 -0.0879 -0.1569>}]
 pizza1 [{shapeType=Box, extentX=14.631, extentY=6.5118, extentZ=32.496, center=<2.5034 -0.2171 0.7044>}]
 pizza2 [{shapeType=Box, extentX=8.4689, extentY=4.8096, extentZ=22.7387, center=<3.6729 0.625 -0.0949>}]
 pizza3 [{shapeType=Box, extentX=6.7626, extentY=4.762, extentZ=15.9316, center=<3.2107 0.2895 0.2938>, rotation=<rot 0 -5.0001 0>}]
 pizza4 [{shapeType=Box, extentX=8.7858, extentY=3.9405, extentZ=11.4387, center=<4.2407 -0.2993 0.1176>, rotation=<rot 0 -5.0001 0>}]
}
constraints [
 {bone1Name=pizza2, bone2Name=pizza1, swing1LimitDegrees=2, swing2LimitDegrees=3.5, twistLimitDegrees=1}
 {bone1Name=pizza3, bone2Name=pizza2, swing1LimitDegrees=2, swing2LimitDegrees=3.5, twistLimitDegrees=1}
 {bone1Name=pizza4, bone2Name=pizza3, swing1LimitDegrees=2, swing2LimitDegrees=3.5, twistLimitDegrees=1}
 {bone1Name=cheeseDripBBBase, bone2Name=pizza4, swing1LimitDegrees=45, swing2LimitDegrees=45, twistLimitDegrees=15}
 {bone1Name=cheeseDripB1End, bone2Name=cheeseDripBBBase, swing1LimitDegrees=45, swing2LimitDegrees=45, twistLimitDegrees=15}
 {bone1Name=cheeseDripB2End, bone2Name=cheeseDripBBBase, swing1LimitDegrees=45, swing2LimitDegrees=45, twistLimitDegrees=15}
 {bone1Name=cheeseDripABase, bone2Name=pizza3, swing1LimitDegrees=45, swing2LimitDegrees=45, twistLimitDegrees=15}
 {bone1Name=cheeseDripAEnd, bone2Name=cheeseDripABase, swing1LimitDegrees=45, swing2LimitDegrees=45, twistLimitDegrees=15}
]
