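# Masked, two-sided lit material for the cheese VFX on the PeppaRonnie back bling, built for Niagara ribbons and mesh particles.
# A rotated, panning smoke texture, randomized per particle, feeds an alpha clip driven by particle alpha, alongside square edge masks.
# World-offset math blends toward a dynamic-parameter position along U and adds vertical droop terms.
# Color math uses particle color and a yellow tint parameter.
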Material /BRCosmetics/Accessories/FORT_Backpacks/Backpack_M_MED_PeppaRonnie/VFX/Materials/M_PeppaRonni_Cheese
graph {
 properties {
  bUsedWithNiagaraMeshParticles=true
  bUsedWithNiagaraRibbons=true
  blendMode=BLEND_Masked
  ditherOpacityMask=true
  opacityMaskClipValue=0.333
  pixelDepthOffsetMode=PDOM_Legacy
  twoSided=true
  usedShadingModels=Default Lit
 }
 nodes [
  Add()
  n1 = DynamicParameter() {paramNames [X,Y,Z,Dip]}
  n2 = AppendVector(A=@n1.X, B=@n1.X)
  n3 = ComponentMask(None=TextureCoordinate()) {r=true}
  LinearInterpolate(B=Subtract(A=AppendVector(A=@n2, B=@n1.Z), B=WorldPosition() {worldPositionShaderOffset=WPT_ExcludeAllShaderOffsets}.XYZ), Alpha=@n3)
  LinearInterpolate(B=ScalarParameter() {defaultValue=-20, parameterName=WPO}, Alpha=ScalarParameter() {parameterName=DroopAmnt})
  n4 = ParticleRandom()
  n5 = Add(A=@n4, B=TextureCoordinate() {uTiling=4, vTiling=0.2})
  n6 = Constant() {r=0.25}
  n7 = Multiply(A=@n4, B=Constant() {r=0.4})
  n8 = Constant() {r=-0.125}
  n9 = LinearInterpolate(A=@n8, B=Abs(None=@n8), Alpha=@n4)
  n10 = TextureSample(UVs=Panner(
    Coordinate=MaterialFunctionCall(UVs=@n5, "Rotation Angle (0-1)"=@n6) {materialFunction=&CustomRotator}.Rotated Values
    Speed=AppendVector(A=@n7, B=@n9)
   ) {speedX=0.3}) {texture=&T_Inky_Smoke_Tile}
  n11 = ParticleColor()
  MaterialFunctionCall!(
   AlphaTexture=@n10.G
   ClipRange=ScalarParameter() {defaultValue=0.1, parameterName=Range}
   AlphaOverTime=LinearInterpolate(Alpha=@n11.A) {constB=0.8}
  ) {materialFunction=/Game/Effects/Fort_Effects/Materials/_Functions/MF_ClipAlpha}
  Multiply(A=@n11.RGB)
  n12 = ScalarParameter() {defaultValue=2, parameterName=Mult2}
  Multiply(
   A=LinearInterpolate(B=Multiply(A=@n10.G, B=@n12), Alpha=ScalarParameter() {parameterName=NoiseAmnt}) {constA=1}
   B=MaterialFunctionCall!(UVs, Mult=ScalarParameter() {defaultValue=1, parameterName=Mult}, Exp=ScalarParameter() {defaultValue=8, parameterName=Exp}) {materialFunction=/Game/Effects/Fort_Effects/MaterialFunctions/MF_SquareEdgeMask}.HorizMask
  )
  n13 = MaterialFunctionCall(UVs=TextureCoordinate() {coordinateIndex=1}, Mult=ScalarParameter() {defaultValue=1.5, parameterName=Exp2}, Exp=Constant() {r=1}) {materialFunction=&MF_SquareEdgeMask}
  Multiply(A=Multiply(A=Saturate(None=@n13.Vertmask), B=@n1.Z), B=Constant3Vector() {constant=<rgba 0 0 1 0>})
  OneMinus(None=@n3)
  VectorParameter() {parameterName=Color, defaultValue=<rgba 1 0.7549 0.1886 0>}
 ]
}
