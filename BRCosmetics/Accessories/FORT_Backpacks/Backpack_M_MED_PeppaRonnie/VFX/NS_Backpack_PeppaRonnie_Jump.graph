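# Niagara jump effect for the PeppaRonnie backpack, with short bursts of melted cheese strands and grease drips as mesh particles.
# Melty_Cheese follows skeletal mesh bones or sockets every frame by particle ID, stretching and thinning through curves varied by a per-particle random that also sets lifetime.
# Greasy_drip combines burst and rate spawning at random indices from four bones or sockets, then falls under randomized downward gravity while shrinking.
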NiagaraSystem /BRCosmetics/Accessories/FORT_Backpacks/Backpack_M_MED_PeppaRonnie/VFX/NS_Backpack_PeppaRonnie_Jump
dependencies {
 usedDataInterfaces [&NiagaraDataInterfaceCurve, &NiagaraDataInterfaceSkeletalMesh]
 usedDynamicInputs []
 usedModules [
  &self:SystemSpawnScript.SystemScriptSource.SystemScriptGraph.NiagaraNodeAssignment_0.SetVariables_#1
  &SystemState
 ]
}
emitters {
 Greasy_drip {
  emitterUpdateScript [
   &EmitterState(Life Cycle Mode=::NewEnumerator1, Loop Behavior=::NewEnumerator1, Loop Duration=0.3, UseLoopDelay=false)
   &SpawnRate(Spawn Probability=0.2, SpawnRate=15, Use Spawn Probability=false)
   &SpawnBurst_Instantaneous(Spawn Count=5, Use Spawn Probability=false)
  ]
  particleSpawnScript [
   /SetVariables(Particles.Custom.Acceleration:NiagaraFloat=&UniformRangedFloat(Minimum=-200, Maximum=-650)) {isSetParameters=true}
   /Niagara/Modules/Spawn/Initialization/InitializeParticle(
    Lifetime=&UniformRangedFloat(Minimum=0.3, Maximum=1)
    Mesh Scale=&VectorFromFloat(Value=&UniformRangedFloat(Minimum=0.5, Maximum=0.8))
    Module.Write Color=true
    Module.Write Lifetime=true
    Module.Write Mass=true
    Module.Write Material Random=true
    Module.Write Position=false
    Module.Write Scale=true
    Module.Write SpriteRotation=false
    Module.Write SpriteSize=false
    Position=&SimulationPosition()
    Sprite Size=<5 5>
   )
   /Game/Effects/Niagara/Modules/Groups/FN_SetSpawnGroup_v1/FN_SetSpawnGroup(Override SpawnGroup=true, SpawnGroup:NiagaraInt32=&UniformRangedInt(Minimum=0, Maximum=3))
   &SkeletalMeshLocation(
    Bone / Socket Index=$Particles.Initial.SpawnGroup
    Inherit Velocity=false
    MaskSpawnGroup=false
    Mesh Sampling Type=::NewEnumerator5
    Sampled Position Offset=<0 0 3>
    Skeletal Mesh=$System.Skeletal Mesh
    Socket Sampling Mode=::NewEnumerator5
   )
  ]
  particleUpdateScript [
   &ParticleState()
   &GravityForce(Gravity=&MakeVector(X=0, Y=0, Z=$Particles.Initial.Custom.Acceleration))
   &SolveForcesAndVelocity(Clamp Velocity=false, Limit Acceleration=false)
   &ScaleMeshSize(Scale Factor=&MakeVector(
     X=&FloatFromCurve(FloatCurve=DataInterface() {propertyValues {Curve {keys [<at 0 6>,<at 1 0>]}}})
     Y=1
     Z=0.345
    ))
  ]
  rendererClasses [&NiagaraMeshRendererProperties]
  renderers [{rendererClass=&NiagaraMeshRendererProperties, rendererIndex=0}]
 }
 Melty_Cheese {
  emitterUpdateScript [
   &EmitterState(Inactive Response=::NewEnumerator1, Life Cycle Mode=::NewEnumerator1, Loop Behavior=::NewEnumerator1, Loop Duration=0.5, UseLoopDelay=false)
   &SpawnBurst_Instantaneous(Spawn Count=4, Use Spawn Probability=false)
  ]
  particleSpawnScript [
   /SetVariables(Particles.MaterialRandom:NiagaraFloat=&UniformRangedFloat(Minimum=0, Maximum=1)) {isSetParameters=true}
   &InitializeParticle(
    Lifetime=&Lerp_Float(A=0.4, B=0.9, Alpha=$Particles.MaterialRandom)
    Material Random:NiagaraFloat=$Particles.MaterialRandom
    Mesh Scale=&UniformRangedVector(Minimum=<0.8 1 0.2>, Maximum=<1.3 1 0.8>)
    Module.Write Color=true
    Module.Write Lifetime=true
    Module.Write Mass=true
    Module.Write Material Random=true
    Module.Write Position=false
    Module.Write Scale=true
    Module.Write SpriteRotation=true
    Module.Write SpriteSize=true
    Position=&SimulationPosition()
    Sprite Rotation=&UniformRangedFloat(Minimum=1, Maximum=3)
    Sprite Size=<5 5>
   )
  ]
  particleUpdateScript [
   &SkeletalMeshLocation(
    Apply Sampled Orientation to Mesh Particles=true
    Bone / Socket Index=/Niagara/DynamicInputs/Execution/ReturnParticleID()
    Inherit Velocity=false
    MaskSpawnGroup=false
    Mesh Sampling Type=::NewEnumerator5
    Skeletal Mesh=$System.Skeletal Mesh
    Socket Sampling Mode=::NewEnumerator5
   )
   &ParticleState()
   &SolveForcesAndVelocity(Clamp Velocity=false, Limit Acceleration=false)
   &ScaleMeshSize(Scale Factor=&MakeVector(
     X=&Multiply_Float(
      A=&Add_Float(A=1, B=$Particles.Initial.MaterialRandom)
      B=&Multiply_Float(
       A=5
       B=&FloatFromCurve(FloatCurve=DataInterface() {propertyValues {Curve {keys [
            {arriveTangent=3.7266, interpMode=RCIM_Cubic, leaveTangent=3.7266, tangentMode=RCTM_User, time=0, value=0}
            {arriveTangent=0.306, interpMode=RCIM_Cubic, leaveTangent=0.306, tangentMode=RCTM_User, time=0.3, value=1}
            {arriveTangent=-1.1547, interpMode=RCIM_Cubic, leaveTangent=-1.1547, tangentMode=RCTM_User, time=1, value=0.3}
           ]}}})
      )
     )
     Y=0.1
     Z=&Multiply_Float(
      A=&Lerp_Float(A=1.65, B=0.6, Alpha=$Particles.Initial.MaterialRandom)
      B=&FloatFromCurve(FloatCurve=DataInterface() {propertyValues {Curve {keys [
           {arriveTangent=-0.1192, interpMode=RCIM_Cubic, leaveTangent=-0.1192, tangentMode=RCTM_User, time=0, value=2}
           {arriveTangent=0.0996, interpMode=RCIM_Cubic, leaveTangent=0.0996, tangentMode=RCTM_User, time=1, value=0.3}
          ]}}})
     )
    ))
   &DynamicMaterialParameters(Module.Param0WriteEnabled=true, Module.Param1WriteEnabled=true, Module.Param2WriteEnabled=true, Module.Param3WriteEnabled=true) {enabled=false}
  ]
  rendererClasses [&NiagaraMeshRendererProperties]
  renderers [{rendererClass=&NiagaraMeshRendererProperties, rendererIndex=0}]
 }
}
summary {
 systemName=NS_Backpack_PeppaRonnie_Jump
 emitters [
  {emitterName=Melty_Cheese, rendererClasses [&NiagaraMeshRendererProperties]}
  {emitterName=Greasy_drip, rendererClasses [&NiagaraMeshRendererProperties]}
 ]
 userVariables [
  User.bIsFrontEndPreview {type=NiagaraBool, defaultValue=true}
  User.RibbonIDOffset {type=NiagaraInt32}
 ]
}
