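# Niagara landing effect for the PeppaRonnie backpack cosmetic that throws off short-lived mesh grease drips and splooshes.
# Three burst emitters spawn mesh particles at bones or sockets of User.SkeletalMesh, picked by a random spawn group index.
# The local and world sploosh emitters add spherical offset and outward plus directional velocity, while drips only fall.
# Per-particle random downward acceleration drives the motion, and curves reshape mesh X and Z scale, shrinking Z to zero.
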NiagaraSystem /BRCosmetics/Accessories/FORT_Backpacks/Backpack_M_MED_PeppaRonnie/VFX/NS_Backpack_PeppaRonnie_Land
dependencies {usedDataInterfaces [&NiagaraDataInterfaceCurve], usedDynamicInputs [], usedModules [&SystemState]}
emitters {
 Greasy_drip {
  emitterUpdateScript [
   &EmitterState(Life Cycle Mode=::NewEnumerator1, Loop Behavior=::NewEnumerator1, Loop Duration=0.3, UseLoopDelay=false)
   &SpawnRate(Spawn Probability=0.2, SpawnRate=10, Use Spawn Probability=false)
   &SpawnBurst_Instantaneous(Spawn Count=5, Use Spawn Probability=false)
  ]
  particleSpawnScript [
   /SetVariables(Particles.Custom.Acceleration:NiagaraFloat=&UniformRangedFloat(Minimum=-200, Maximum=-650)) {isSetParameters=true, module=SetVariables_#1}
   /Niagara/Modules/Spawn/Initialization/InitializeParticle(
    Lifetime=&UniformRangedFloat(Minimum=0.3, Maximum=0.5)
    Mesh Scale=&VectorFromFloat(Value=&UniformRangedFloat(Minimum=0.7, Maximum=1))
    Module.Write Color=true
    Module.Write Lifetime=true
    Module.Write Mass=true
    Module.Write Material Random=true
    Module.Write Position=false
    Module.Write Scale=true
    Module.Write SpriteRotation=false
    Module.Write SpriteSize=false
    Position=&SimulationPosition()
    Sprite Size=<5 5>
   )
   /Game/Effects/Niagara/Modules/Groups/FN_SetSpawnGroup_v1/FN_SetSpawnGroup(Override SpawnGroup=true, SpawnGroup:NiagaraInt32=&UniformRangedInt(Minimum=0, Maximum=3))
   &SkeletalMeshLocation(
    Bone / Socket Index=$Particles.Initial.SpawnGroup
    Inherit Velocity=false
    MaskSpawnGroup=false
    Mesh Sampling Type=::NewEnumerator5
    Skeletal Mesh=$User.SkeletalMesh
    Socket Sampling Mode=::NewEnumerator5
   )
  ]
  particleUpdateScript [
   &ParticleState()
   &GravityForce(Gravity=&MakeVector(X=0, Y=0, Z=$Particles.Initial.Custom.Acceleration))
   &SolveForcesAndVelocity(Clamp Velocity=false, Limit Acceleration=false)
   &ScaleMeshSize(Scale Factor=&MakeVector(
     X=&FloatFromCurve(FloatCurve=DataInterface() {propertyValues {Curve {keys [<at 0 5>,<at 1 0.3>]}}})
     Y=1
     Z=&FloatFromCurve(FloatCurve=DataInterface() {propertyValues {Curve {keys [<at 0 1 RCIM_Cubic>, {arriveTangent=-1.6318, interpMode=RCIM_Cubic, leaveTangent=-1.6318, tangentMode=RCTM_User, time=1, value=0}]}}})
    ))
  ]
  rendererClasses [&NiagaraMeshRendererProperties]
  renderers [{rendererClass=&NiagaraMeshRendererProperties, rendererIndex=0}]
 }
 Greasy_sploosh_local {
  emitterUpdateScript [
   &EmitterState(Inactive Response=::NewEnumerator1, Life Cycle Mode=::NewEnumerator1, Loop Behavior=::NewEnumerator1, Loop Duration=0.7, UseLoopDelay=false)
   &SpawnRate(Spawn Probability=0.2, SpawnRate=25, Use Spawn Probability=true) {enabled=false}
   &SpawnBurst_Instantaneous(Spawn Count=3, Use Spawn Probability=false)
  ]
  particleSpawnScript [
   /SetVariables(Particles.Custom.Acceleration:NiagaraFloat=&UniformRangedFloat(Minimum=-400, Maximum=-1200)) {isSetParameters=true, module=SetVariables_#1}
   &InitializeParticle(
    Lifetime=&UniformRangedFloat(Minimum=0.225, Maximum=0.6)
    Mesh Scale=&VectorFromFloat(Value=&UniformRangedFloat(Minimum=0.5, Maximum=1))
    Module.Write Color=true
    Module.Write Lifetime=true
    Module.Write Mass=true
    Module.Write Material Random=true
    Module.Write Position=false
    Module.Write Scale=true
    Module.Write SpriteRotation=false
    Module.Write SpriteSize=false
    Position=&SimulationPosition()
    Sprite Size=<5 5>
   )
   &FN_SetSpawnGroup(Override SpawnGroup=true, SpawnGroup:NiagaraInt32=&UniformRangedInt(Minimum=0, Maximum=3))
   &SkeletalMeshLocation(
    Bone / Socket Index=$Particles.Initial.SpawnGroup
    Inherit Velocity=false
    MaskSpawnGroup=false
    Mesh Sampling Type=::NewEnumerator5
    Skeletal Mesh=$User.SkeletalMesh
    Socket Sampling Mode=::NewEnumerator5
   )
   /Niagara/Modules/Spawn/Location/SphereLocation(MaskSpawnGroup=false, Module.Override Randomness=false, Non Uniform Scale=<1 1 1.5>, Offset=<0 0 8>, Sphere Radius=3, Surface Only=false)
   &AddVelocityFromPoint(Module.Use Falloff Distance=false, Velocity Strength=&UniformRangedFloat(Minimum=0, Maximum=70))
   &AddVelocity(Scale Added Velocity=&UniformRangedVector(Minimum=<-1 -1 0.2>, Maximum=<1 0 1>), Velocity=<50 0 -100>)
  ]
  particleUpdateScript [
   &ParticleState()
   &GravityForce(Gravity=&MakeVector(X=0, Y=0, Z=$Particles.Initial.Custom.Acceleration)) {enabled=false}
   &AccelerationForce(
    Acceleration=&MakeVector(X=0, Y=0, Z=$Particles.Initial.Custom.Acceleration)
    Coordinate Space=::Simulation
   )
   &SolveForcesAndVelocity(Clamp Velocity=false, Limit Acceleration=false)
   &ScaleMeshSize(Scale Factor=&MakeVector(
     X=&FloatFromCurve(FloatCurve=DataInterface() {propertyValues {Curve {keys [
          {arriveTangent=0.2579, interpMode=RCIM_Cubic, leaveTangent=0.2579, tangentMode=RCTM_User, time=0, value=2}
          {arriveTangent=0.9286, interpMode=RCIM_Cubic, leaveTangent=0.9286, tangentMode=RCTM_User, time=1, value=3}
         ]}}})
     Y=1
     Z=&FloatFromCurve(FloatCurve=DataInterface() {propertyValues {Curve {keys [<at 0 1 RCIM_Cubic>, {arriveTangent=-1.6027, interpMode=RCIM_Cubic, leaveTangent=-1.6027, tangentMode=RCTM_User, time=1, value=0}]}}})
    ))
  ]
  rendererClasses [&NiagaraMeshRendererProperties]
  renderers [{rendererClass=&NiagaraMeshRendererProperties, rendererIndex=0}]
 }
 Greasy_sploosh_world {
  emitterUpdateScript [
   &EmitterState(Inactive Response=::NewEnumerator1, Life Cycle Mode=::NewEnumerator1, Loop Behavior=::NewEnumerator1, Loop Duration=0.7, UseLoopDelay=false)
   &SpawnRate(Spawn Probability=0.2, SpawnRate=25, Use Spawn Probability=true) {enabled=false}
   &SpawnBurst_Instantaneous(Spawn Count=3, Use Spawn Probability=false)
  ]
  particleSpawnScript [
   /SetVariables(Particles.Custom.Acceleration:NiagaraFloat=&UniformRangedFloat(Minimum=-400, Maximum=-1200)) {isSetParameters=true, module=SetVariables_#1}
   &InitializeParticle(
    Lifetime=&UniformRangedFloat(Minimum=0.225, Maximum=0.6)
    Mesh Scale=&VectorFromFloat(Value=&UniformRangedFloat(Minimum=0.5, Maximum=0.7))
    Module.Write Color=true
    Module.Write Lifetime=true
    Module.Write Mass=true
    Module.Write Material Random=true
    Module.Write Position=false
    Module.Write Scale=true
    Module.Write SpriteRotation=false
    Module.Write SpriteSize=false
    Position=&SimulationPosition()
    Sprite Size=<5 5>
   )
   &FN_SetSpawnGroup(Override SpawnGroup=true, SpawnGroup:NiagaraInt32=&UniformRangedInt(Minimum=0, Maximum=3))
   &SkeletalMeshLocation(
    Bone / Socket Index=$Particles.Initial.SpawnGroup
    Inherit Velocity=false
    MaskSpawnGroup=false
    Mesh Sampling Type=::NewEnumerator5
    Skeletal Mesh=$User.SkeletalMesh
    Socket Sampling Mode=::NewEnumerator5
   )
   &SphereLocation(MaskSpawnGroup=false, Module.Override Randomness=false, Non Uniform Scale=<1 1 1.5>, Offset=<0 0 8>, Sphere Radius=3, Surface Only=false)
   &AddVelocityFromPoint(Module.Use Falloff Distance=false, Velocity Strength=&UniformRangedFloat(Minimum=0, Maximum=70))
   &AddVelocity(Scale Added Velocity=&UniformRangedVector(Minimum=<-1 -1 0.2>, Maximum=<1 0 1>), Velocity=<120 0 -100>)
  ]
  particleUpdateScript [
   &ParticleState()
   &GravityForce(Gravity=&MakeVector(X=0, Y=0, Z=$Particles.Initial.Custom.Acceleration)) {enabled=false}
   &AccelerationForce(
    Acceleration=&MakeVector(X=0, Y=0, Z=$Particles.Initial.Custom.Acceleration)
    Coordinate Space=::Simulation
   )
   &SolveForcesAndVelocity(Clamp Velocity=false, Limit Acceleration=false)
   &ScaleMeshSize(Scale Factor=&MakeVector(
     X=&FloatFromCurve(FloatCurve=DataInterface() {propertyValues {Curve {keys [
          {arriveTangent=0.2579, interpMode=RCIM_Cubic, leaveTangent=0.2579, tangentMode=RCTM_User, time=0, value=3}
          {arriveTangent=0.9286, interpMode=RCIM_Cubic, leaveTangent=0.9286, tangentMode=RCTM_User, time=1, value=2}
         ]}}})
     Y=1
     Z=&FloatFromCurve(FloatCurve=DataInterface() {propertyValues {Curve {keys [<at 0 1>,<at 1 0>]}}})
    ))
  ]
  rendererClasses [&NiagaraMeshRendererProperties]
  renderers [{rendererClass=&NiagaraMeshRendererProperties, rendererIndex=0}]
 }
}
summary {
 systemName=NS_Backpack_PeppaRonnie_Land
 emitters [
  {emitterName=Greasy_sploosh_local, rendererClasses [&NiagaraMeshRendererProperties]}
  {emitterName=Greasy_sploosh_world, rendererClasses [&NiagaraMeshRendererProperties]}
  {emitterName=Greasy_drip, rendererClasses [&NiagaraMeshRendererProperties]}
 ]
 userVariables [
  User.bIsFrontEndPreview {type=NiagaraBool, defaultValue=true}
  User.RibbonIDOffset {type=NiagaraInt32}
  User.SkeletalMesh {
   type=NiagaraDataInterfaceSkeletalMesh
   defaultValue {
    struct=NiagaraExt_VariableValue_DataInterface
    value {dataInterface=&self:NiagaraDataInterfaceSkeletalMesh_2}
   }
  }
 ]
}
